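; Parent gameplay ability that watches an AI bot pawn for broken shields and triggers a shield-break ability by tag.
; It binds to pawn damage and controller alert-level changes, checking for depleted shields and Threatened alert.
; On a break, it applies loose tags and a GameplayEffect, then attempts activation, retrying on a looping timer.
; It ends upon success, existing granted tag or vehicle detection on damage, clearing timer and bindings.
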(event Ability|EventActivateAbility
  ;[1 1 1] Activate Ability
  (Development|PrintString "@@@ Event ActivateAbility [GA_ShieldBreakMonitor_Parent]: CALLED!" false (Variables|Debug|GetbDebugGameplayAbility))
  (bind _outfortplayerpawn (CallFunction|GetAvatarasValidFortPlayerPawn)
    (:ValidFortPlayerPawn
      (bind _output_get (Variables|InternalData|SetAIPawn _outfortplayerpawn))
      (EventDispatchers|BindEventtoOnDamaged _output_get (EventDispatchers|CreateEvent :!event Custom|OnAIPawnDamaged))
      (bind _asfort_athena_aibot_controller (Utilities|Casting|CastToFortAthenaAIBotController (AI|GetAIController _output_get))
        (:then
          (bind _output_get_2 (Variables|InternalData|SetAIBotController _asfort_athena_aibot_controller))
          (AlertLevel|BindEventtoOnAlertLevelChangedEvent _output_get_2 (EventDispatchers|CreateEvent :!event Custom|AlertLevelChanged))
          (Development|PrintString "@@@@@@ Event ActivateAbility [GA_ShieldBreakMonitor_Parent]: Shield event bound..." false (Variables|Debug|GetbDebugGameplayAbility)))
        (:CastFailed
          (Ability|EndAbility))))
    (:AvatarCastFailed
      (Ability|EndAbility))))

(event Custom|OnAIPawnDamaged
  (params (DamagedActor :type Actor%20Object%20Reference) (Damage :type Float%20%28single-precision%29) (InstigatedBy :type Controller%20Object%20Reference) (DamageCauser :type Actor%20Object%20Reference) (HitLocation :type Vector) (FHitComponent :type Primitive%20Component%20Object%20Reference) (BoneName :type Name) (Momentum :type Vector))
  ;[1 1 1] Validate that we can try to activate the new Ability
  (if (CallFunction|ValidateShields)
    (if (GameplayTags|HasMatchingGameplayTag (Variables|InternalData|GetAIPawn) "(TagName=\"Gameplay.Action.Player.AthenaInVehicle\")")
      (Ability|EndAbility)
      (elif (Variables|Default|GetNeedsToBeThreatenedToActivateAbility)
        (if (Utilities|Enum|Equal(Enum) (Variables|/Script/FortniteAI.FortAthenaAIBotController|GetCurrentAlertLevel (Variables|InternalData|GetAIBotController)) "Threatened")
          (if (not (Utilities|Time|IsValidTimerHandle (Variables|InternalData|GetRetryActivateNewAbilityTimer)))
            (CallFunction|ApplyTagstoAIPawn)
            (CallFunction|ApplyGameplayEffecttoAIPawn)
            (CallFunction|TryActivateNewAbility)))
        (else
          _)))))

(event Custom|TryActivateNewAbility
  ;[1 1 1] Try to Activate Ability
  (Development|PrintString "@@@@@@ PawnShieldBreak [GA_ShieldBreakMonitor_Parent]: CALLED!" false (Variables|Debug|GetbDebugGameplayAbility))
  (if (GameplayTags|HasMatchingGameplayTag (Variables|InternalData|GetAIPawn) (Variables|Default|GetTag_GrantedTagByNewAbility))
    (Ability|EndAbility)
    (else
      (bind _returnvalue (Abilities|TryActivateAbilitiesbyTag (Ability|GetAbilitySystemComponent (Variables|InternalData|GetAIPawn)) (Variables|Default|GetTag_ShieldBreakAbilityTag) false))
      (if _returnvalue
        (Ability|EndAbility)
        (else
          (Fortnite|FortShippingLog (Variables|InternalData|GetAIPawn) (Utilities|String|ToString(Text) (Utilities|Text|FormatText "NSLOCTEXT(\"[6C334FC936864D7A54D41BC9FE9A6E2C]\", \"1F0A43034B41676B9AA8E69B7B6735C2\", \"Pawn {a} tried to activate Shield Break Ability with Tag : {b} but ability could not be activated. Make sure a Tag has been set and the Ability can be used by this Pawn.\")" (Utilities|GetDisplayName (Variables|InternalData|GetAIPawn)) (GameplayTags|GetDebugStringfromGameplayTagContainer (Variables|Default|GetTag_ShieldBreakAbilityTag)))) true)
          (bind _returnvalue_2 (Utilities|Time|SetTimerbyEvent (EventDispatchers|CreateEvent :!event Custom|TryActivateNewAbility) 0.2 true true))
          (Variables|InternalData|SetRetryActivateNewAbilityTimer _returnvalue_2))))))

(event Custom|AlertLevelChanged
  (params (BotController :type Fort%20Athena%20AIBot%20Controller%20Object%20Reference) (OldAlertLevel :type EAlertLevel%20Enum :default "Unaware") (NewAlertLevel :type EAlertLevel%20Enum :default "Unaware"))
  ;[1 1 1] Validate that we can try to activate the new Ability
  (switch Utilities|FlowControl|Switch|SwitchonEAlertLevel NewAlertLevel
    (:Threatened
      (if (CallFunction|ValidateShields)
        (if (not (Utilities|Time|IsValidTimerHandle (Variables|InternalData|GetRetryActivateNewAbilityTimer)))
          (CallFunction|ApplyTagstoAIPawn)
          (CallFunction|ApplyGameplayEffecttoAIPawn)
          (CallFunction|TryActivateNewAbility))))))

(event Custom|ApplyTagsToAIPawn
  ; Try to apply some Gameplay Tags to the AI when the shields are broken.
  (Ability|GameplayEffect|AddLooseGameplayTags (Variables|InternalData|GetAIPawn) (Variables|Default|GetLooseTagsOnShieldBreak) true))

(event Custom|ApplyGameplayEffectToAIPawn
  (bind _aipawn (Variables|InternalData|GetAIPawn))
  (bind _returnvalue (Ability|GetAbilitySystemComponent _aipawn))
  ; Try to apply a GameplayEffect to the AI when the shields are broken.
  (Utilities|IsValid _returnvalue
    (:Is%20Valid
      (GameplayEffects|ApplyGameplayEffectToTarget :self _returnvalue :GameplayEffectClass (Variables|Default|GetGE_OnShieldBreak) :Target _returnvalue :Level 0.0))))

(event Ability|EventOnEndAbility
  (params (bWasCancelled :type Boolean))
  (bind _aibotcontroller (Variables|InternalData|GetAIBotController))
  (bind _aipawn (Variables|InternalData|GetAIPawn))
  ;[1 1 1] End ABility
  (Development|PrintString "@@@@@@ Event onEndAbility [GA_ShieldBreakMonitor_Parent]: CALLED!" false (Variables|Debug|GetbDebugGameplayAbility))
  (Utilities|Time|ClearandInvalidateTimerbyHandle (Variables|InternalData|GetRetryActivateNewAbilityTimer))
  (Utilities|IsValid _aipawn
    (:Is%20Valid
      (EventDispatchers|UnbindEventfromOnDamaged _aipawn (EventDispatchers|CreateEvent :!event Custom|OnAIPawnDamaged)))
    (:Is%20Not%20Valid
      _))
  (Utilities|IsValid _aibotcontroller
    (:Is%20Valid
      (AlertLevel|UnbindEventfromOnAlertLevelChangedEvent _aibotcontroller (EventDispatchers|CreateEvent :!event Custom|AlertLevelChanged)))))

(fn ValidateShields
  (return (and (<= (FortPawn|GetShield (Variables|InternalData|GetAIPawn)) 0) (> (FortPawn|GetMaxShield (Variables|InternalData|GetAIPawn)) 0))))
